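# Niagara idle effect for the StrideMice guitar cosmetic, combining twinkling flare sprites with a blue outer glow.
# Alt_Flares spawns sprites by bursts and probabilistic rate at six fixed offsets, pulsing their size and randomizing hue and brightness.
# OuterGlow002 holds a looping-lifetime particle on two mesh renderers, animating a dynamic material parameter.
# Glow colors follow system time-of-day color and curves, and front-end flags adjust glow and flare intensity.
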NiagaraSystem /SparksCosmetics/Instrument/Guitar/StrideMice/Effects/NS_Guitar_StrideMice_IdleFX
dependencies {
 usedDataInterfaces [
  &NiagaraDataInterfaceArrayFloat3
  &NiagaraDataInterfaceCurve
  &NiagaraDataInterfaceColorCurve
  &NiagaraDataInterfaceCamera
 ]
 usedDynamicInputs [
  &ColorFromCurve
  &Abs_Float
  &DotProduct
  &ReturnCameraVector
  &MakeFloatFromVector
  &Lerp_Float
 ]
 usedModules [
  &SystemState
  &self:SystemSpawnScript.SystemScriptSource.SystemScriptGraph.NiagaraNodeAssignment_0.SetVariables_#1
  &self:SystemSpawnScript.SystemScriptSource.SystemScriptGraph.NiagaraNodeAssignment_4.SetVariables_#2
  &self:SystemSpawnScript.SystemScriptSource.SystemScriptGraph.NiagaraNodeAssignment_3.SetVariables_#3
  &self:SystemSpawnScript.SystemScriptSource.SystemScriptGraph.NiagaraNodeAssignment_1.SetVariables_#4
  &self:SystemSpawnScript.SystemScriptSource.SystemScriptGraph.NiagaraNodeAssignment_2.SetVariables_#5
  &self:SystemSpawnScript.SystemScriptSource.SystemScriptGraph.NiagaraNodeAssignment_6.SetVariables_#6
 ]
}
emitters {
 Alt_Flares {
  emitterSpawnScript [
   /SetVariables(Emitter.FlareColor:LinearColor=<rgba 0.164 0.202 0.3125 3>) {isSetParameters=true}
   /SetVariables(Emitter.SpawnCount=6) {isSetParameters=true}
  ]
  emitterUpdateScript [
   &EmitterState()
   &SpawnRate(Spawn Probability=0.75, SpawnRate=2, Use Spawn Probability=true) {module=SpawnRate001}
   &SpawnBurst_Instantaneous(Spawn Count=3, Use Spawn Probability=false, UseLoopCountLimit=true)
   &SpawnBurst_Instantaneous(Spawn Count=3, Spawn Time=3.3333, Use Spawn Probability=false, UseLoopCountLimit=false) {module=SpawnBurst_Instantaneous001}
  ]
  particleSpawnScript [
   &PartitionParticles(Number Of Partitions=$Emitter.SpawnCount)
   &InitializeParticle(
    Lifetime=&RandomRangeFloat(Minimum=1.5, Maximum=2.5)
    Position Offset=&SelectVectorFromArray(
     Direct Array Index=$Particles.PartitionParticles.Partition
     Direct Array Mode=::NewEnumerator1
     Vector Selection Array=DataInterface() {propertyValues {FloatData [<-4.8 -4.15 66>, <20 -0.5 -24>, <-11.09 -0.5 16.92>, <-7 -0.5 -20>, <18.25 -0.5 -3.75>, <18 -3.25 52>]}}
    )
    Sprite Rotation Angle=25
    Sprite Rotation Mode=::NewEnumerator0
    Sprite Size Mode=::NewEnumerator3
    Uniform Sprite Size=&FloatFromCurve(
     FloatCurve=DataInterface() {propertyValues {Curve {keys [<at 0 9>,<at 0.85 14>,<at 1 20>]}}}
     CurveIndex=&RandomRangeFloat(Minimum=0, Maximum=1)
    )
    UsePositionOffset=true
    Write Material Random=true
   )
   &CameraOffset(Camera Offset Amount=7)
  ]
  particleUpdateScript [
   &ParticleState()
   &ScaleSpriteSize(Uniform Curve Sprite Scale=DataInterface() {propertyValues {Curve {
       defaultValue=0
       keys [
        {arriveTangent=9.0725, interpMode=RCIM_Cubic, leaveTangent=9.0725, tangentMode=RCTM_User, time=0, value=0.5}
        {arriveTangent=-2.5249, interpMode=RCIM_Cubic, leaveTangent=-2.5249, tangentMode=RCTM_User, time=0.0921, value=1.2018}
        {arriveTangent=-0.4898, interpMode=RCIM_Cubic, leaveTangent=-0.4898, time=0.2331, value=0.7184}
        {arriveTangent=-0.1736, interpMode=RCIM_Cubic, leaveTangent=-0.1736, time=0.3693, value=1.0661}
        {arriveTangent=-0.1374, interpMode=RCIM_Cubic, leaveTangent=-0.1374, time=0.601, value=0.6545}
        {arriveTangent=-1.6404, interpMode=RCIM_Cubic, leaveTangent=-1.6404, time=0.85, value=1}
        {arriveTangent=-9.0647, interpMode=RCIM_Cubic, leaveTangent=-9.0647, tangentMode=RCTM_User, time=1, value=0}
       ]
      }}}) {module=ScaleSpriteSize001}
   &ScaleColor(Scale RGB=<0.2 0.2 0.2>, ScaleA=true, ScaleRGB=true)
   /SetVariables(Particles.ChromaFlare:LinearColor=&Multiply_LinearColorByFloat(Linear Color=$Particles.Color, Float=&RandomRangeFloat(Minimum=0.13, Maximum=0.16))) {isSetParameters=true}
   /SetVariables(Particles.FlareColor=&MultiplyLinearColor(
     Linear Color=&HueShiftLinearColor(
      Hue=&RandomRangeFloat(Minimum=-0.02, Maximum=0.02)
      Saturation=&RandomRangeFloat(Minimum=1, Maximum=1.8)
      Value=&RandomRangeFloat(Minimum=0.8, Maximum=1)
      Linear Color=<rgba 0.1464 0.2768 0.5073 3>
     )
     Scale Factor=&MakeVector4FromVectorAndFloat(
      "Vector (RGB)"=<1 1 1>
      "Float (Alpha)"=&Multiply_Float(
       A=&RandomRangeFloat(Minimum=0.3, Maximum=&FN_FrontEndFloatModifier(Gameplay=0.65, Front End=0.5))
       B=&MakeFloatFromLinearColor(LinearColor=$Particles.Color, Channel:ENiagaraFloatFromLinearColorOptions=::NewEnumerator3)
      )
     )
    )) {isSetParameters=true}
  ]
  rendererClasses [&NiagaraSpriteRendererProperties]
  renderers [
   {rendererClass=&NiagaraSpriteRendererProperties}
   {rendererClass=&NiagaraSpriteRendererProperties, rendererIndex=1}
  ]
 }
 OuterGlow002 {
  emitterUpdateScript [
   &EmitterState()
   &SpawnBurst_Instantaneous(Use Spawn Probability=false, UseLoopCountLimit=true)
  ]
  particleSpawnScript [&InitializeParticle(
    Color=<rgba 0.0816 0.0985 0.4062 1>
    Lifetime=9
    Mesh Scale Mode=::NewEnumerator3
    Sprite Size Mode=::NewEnumerator3
    UsePositionOffset=false
    Write Material Random=true
   )]
  particleUpdateScript [
   &ParticleState(
    Kill Particles When Lifetime Has Elapsed=false
    Let Infinitely Lived Particles Die When Emitter Deactivates=true
    Loop Particles Lifetime=true
   )
   &SolveForcesAndVelocity(Clamp Velocity=false, Limit Acceleration=false)
   &DynamicMaterialParameters(
    Index 0 Param 1=&FloatFromCurve(FloatCurve=DataInterface() {propertyValues {Curve {keys [<at 0 1>,<at 1 0>]}}})
    Module.Index0Float0Write=true
    Module.Index0Float1Write=true
    Module.Index0Float2Write=true
    Module.Index0Float3Write=true
   )
   /SetVariables(
    Particles.Color2=&MakeLinearColorFromVectorAndFloat(
     "Vector (RGB)"=&Multiply_Vector(
      A=&Multiply_Vector(
       A=&MakeVectorFromLinearColorRGB(Color=<rgba 0.06 0.5728 2 1.5>)
       B=&MakeVectorFromLinearColorRGB(Color=$System.TOD Color)
      )
      B=&VectorFromFloat(Value=3)
     )
     "Float (Alpha)"=&Multiply_Float(
      A=&MakeCustomFloatFromBool(Bool=$User.bIsFrontEndPreview, True Float=0.3, False Float=0.5)
      B=&FloatFromCurve(
       FloatCurve=DataInterface() {propertyValues {Curve {keys [
           {arriveTangent=0.1117, interpMode=RCIM_Cubic, leaveTangent=0.1117, tangentMode=RCTM_User, time=0, value=1}
           {
            arriveTangent=-0.0076
            arriveTangentWeight=0.0963
            interpMode=RCIM_Cubic
            leaveTangent=-0.0076
            leaveTangentWeight=0.1693
            tangentMode=RCTM_User
            time=0.2
            value=1.25
           }
           {arriveTangentWeight=0.0722, interpMode=RCIM_Cubic, leaveTangentWeight=0.1444, tangentMode=RCTM_User, time=0.6, value=1.25}
           <at 1 1 RCIM_Cubic>
          ]}}}
       CurveIndex=Linked() {linkedVariable {name=NPC.FortniteNPC.TimeOfDay_0-1, type=NiagaraFloat}}
      )
     )
    )
    Particles.Color3:LinearColor=&MakeLinearColorFromVectorAndFloat(
     "Vector (RGB)"=&Multiply_Vector(
      A=&Multiply_Vector(
       A=&MakeVectorFromLinearColorRGB(Color=<rgba 0.08 0.5875 2 1.5>)
       B=&MakeVectorFromLinearColorRGB(Color=$System.TOD Color)
      )
      B=&VectorFromFloat(Value=&FloatFromCurve(
        FloatCurve=DataInterface() {propertyValues {Curve {keys [
            {arriveTangent=0.1117, interpMode=RCIM_Cubic, leaveTangent=0.1117, tangentMode=RCTM_User, time=0, value=1.35}
            {
             arriveTangent=-0.0076
             arriveTangentWeight=0.0963
             interpMode=RCIM_Cubic
             leaveTangent=-0.0076
             leaveTangentWeight=0.1693
             tangentMode=RCTM_User
             time=0.2
             value=1.5
            }
            {arriveTangentWeight=0.0722, interpMode=RCIM_Cubic, leaveTangentWeight=0.1444, tangentMode=RCTM_User, time=0.6, value=1.5}
            <at 1 1.35 RCIM_Cubic>
           ]}}}
        CurveIndex=Linked() {linkedVariable {name=NPC.FortniteNPC.TimeOfDay_0-1, type=NiagaraFloat}}
       ))
     )
     "Float (Alpha)"=&Multiply_Float(
      A=&MakeCustomFloatFromBool(Bool=$User.bIsFrontEndPreview, True Float=0.35, False Float=0.45)
      B=&Multiply_Float(
       A=1.75
       B=&FloatFromCurve(
        FloatCurve=DataInterface() {propertyValues {Curve {keys [
            {arriveTangent=0.1117, interpMode=RCIM_Cubic, leaveTangent=0.1117, tangentMode=RCTM_User, time=0, value=1}
            {
             arriveTangent=-0.0076
             arriveTangentWeight=0.0963
             interpMode=RCIM_Cubic
             leaveTangent=-0.0076
             leaveTangentWeight=0.1693
             tangentMode=RCTM_User
             time=0.2
             value=1.25
            }
            {arriveTangentWeight=0.0722, interpMode=RCIM_Cubic, leaveTangentWeight=0.1444, tangentMode=RCTM_User, time=0.6, value=1.25}
            <at 1 1 RCIM_Cubic>
           ]}}}
        CurveIndex=Linked() {linkedVariable {name=NPC.FortniteNPC.TimeOfDay_0-1, type=NiagaraFloat}}
       )
      )
     )
    )
   ) {isSetParameters=true}
  ]
  rendererClasses [&NiagaraMeshRendererProperties]
  renderers [
   {rendererClass=&NiagaraMeshRendererProperties, rendererIndex=0}
   {rendererClass=&NiagaraMeshRendererProperties, rendererIndex=1}
  ]
 }
}
summary {
 systemName=NS_Guitar_StrideMice_IdleFX
 emitters [
  {emitterName=Alt_Flares, rendererClasses [&NiagaraSpriteRendererProperties]}
  {emitterName=OuterGlow002, rendererClasses [&NiagaraMeshRendererProperties]}
 ]
 userVariables [
  User.bIsFrontEnd {type=NiagaraBool}
  User.bIsFrontEndPreview {type=NiagaraBool}
 ]
}
